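# Translucent UI master material for feature buttons, with a layered color gradient and an optional animated shine swipe.
# A vertical gradient between two colors blends into a second, lighter-by-default gradient through a mask built from two independently rotated, soft-edged axis masks on the 9-slice UVs.
# The shine is a rotated, scrolling sine band added on top, gated by time into on/off cycles scaled by shine speed and faded in by UseShineSwipe.
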Material /FortUILibrary/Foundation/Buttons/Materials/Feature/M_FeatureButton_Master
graph {
 properties {
  bUsedWithStaticMesh=false
  blendMode=BLEND_Translucent
  materialDomain=MD_UI
  pixelDepthOffsetMode=PDOM_Legacy
  usedShadingModels=Default Lit
 }
 nodes [
  n1 = VectorParameter() {group=ButtonSettings, parameterName=vRGB_2, defaultValue=<rgba 1 1 0.2122 1>}
  n2 = MaterialFunctionCall() {materialFunction=&GetUserInterfaceUV}
  n3 = Add(
   A=Divide(A=ComponentMask(None=@n2.9-Slice UV) {g=true}, B=ScalarParameter() {defaultValue=1, group=ButtonSettings, parameterName=sScale})
   B=ScalarParameter() {group=ButtonSettings, parameterName=sPosition}
  )
  n4 = LinearInterpolate(A=VectorParameter() {group=ButtonSettings, parameterName=vRGB_1, defaultValue=<rgba 0.8714 0.7157 0 1>}.RGB, B=@n1.RGB, Alpha=@n3)
  n5 = VectorParameter() {group=ButtonSettings, parameterName=vRGB_3, defaultValue=<rgba 1 1 1 1>}
  n6 = Add(A=Rotator(Coordinate=@n2.9-Slice UV, Time=ScalarParameter() {defaultValue=0.8, group=ButtonSettings, parameterName=sRotation_G})) {constB=-0.5}
  n7 = ScalarParameter() {defaultValue=0.4, group=ButtonSettings, parameterName=sOffset_G}
  n8 = ScalarParameter() {defaultValue=0.05, group=ButtonSettings, parameterName=sSSMargin, sliderMax=1}
  n9 = MaterialFunctionCall(Alpha=ComponentMask(None=Abs(None=@n6)) {g=true}, Min=@n7, Max=Add(A=@n7, B=@n8)) {materialFunction=&SmoothStep}
  n10 = Add(A=Rotator(Coordinate=@n2.9-Slice UV, Time=ScalarParameter() {defaultValue=0.8, group=ButtonSettings, parameterName=sRotation_R})) {constB=-0.5}
  n11 = ScalarParameter() {defaultValue=0.4, group=ButtonSettings, parameterName=sOffset_R}
  n12 = MaterialFunctionCall(Alpha=ComponentMask(None=Abs(None=@n10)) {r=true}, Min=@n11, Max=Add(A=@n11, B=Multiply(A=@n8) {constB=0.3})) {materialFunction=&SmoothStep}
  n13 = LinearInterpolate(
   A=@n4
   B=LinearInterpolate(A=@n4, B=LinearInterpolate(A=@n1.RGB, B=@n5.RGB, Alpha=@n3), Alpha=LinearInterpolate(A=@n9.Result, Alpha=@n12.Result))
   Alpha=ScalarParameter() {defaultValue=1, group=ButtonSettings, parameterName=sTextMaterial, sliderMax=1}
  )
  n14 = ScalarParameter() {defaultValue=0.15, group=Shine, parameterName=Shine Rotation}
  n15 = ScalarParameter() {defaultValue=4, group=Shine, parameterName=Shine Speed}
  n16 = OneMinus(None=Multiply(A=Time(), B=@n15))
  n17 = Constant()
  n18 = ComponentMask(None=Add(A=Rotator(Coordinate=@n2.9-Slice UV, Time=@n14) {centerX=0, centerY=0}, B=AppendVector(A=@n16, B=@n17))) {r=true}
  n19 = ScalarParameter() {defaultValue=0.75, group=Shine, parameterName=Shine SS Min}
  n20 = ScalarParameter() {defaultValue=0.25, group=Shine, parameterName=Shine SS Max}
  n21 = LinearInterpolate(
   A=Constant()
   B=VectorParameter() {group=Shine, parameterName=ShineRGBA, defaultValue=<rgba 1 1 1 0.5>}.RGB
   Alpha=MaterialFunctionCall(Alpha=Sine(None=@n18) {period=2}, Min=@n19, Max=Add(A=@n19, B=@n20)) {materialFunction=&SmoothStep}.Result
  )
  n22 = Constant() {r=1}
  n23 = Time()
  n24 = Divide(A=ScalarParameter() {defaultValue=3, group=Shine, parameterName=Cycles On}, B=@n15)
  n25 = Constant() {r=2}
  n26 = Multiply(A=Divide(A=ScalarParameter() {defaultValue=6, group=Shine, parameterName=Cycles Off}, B=@n15), B=Constant() {r=2})
  n27 = Fmod(A=@n23, B=Add(A=Multiply(A=@n24, B=@n25), B=@n26))
  n28 = Constant()
  n29 = Constant() {r=-1}
  n30 = If(
   A=Add(A=Fmod(A=@n23, B=@n22), B=If(A=@n27, B=@n26, "A > B"=@n28, "A == B"=@n29, "A < B"=@n29)) {constB=-1}
   B=Constant()
   "A > B"=Constant() {r=1}
   "A < B"=Constant()
  )
  LinearInterpolate(A=@n13, B=Add(A=Multiply(A=@n21, B=@n30), B=@n13), Alpha=ScalarParameter() {parameterName=UseShineSwipe, sliderMax=1})
 ]
}
